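; Burst gameplay cue for the Nitro Gauntlets ground punch, adding smoke and a physics force effect at the punching player.
; Where cosmetic events can run, it spawns a Niagara smoke system at the target's transform and passes in whether the pawn is moving on ground.
; If the physics field console variable is enabled, it also spawns a radial force field actor ahead of the target along its forward vector.
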(fn OnBurst
  (params (Target :type Actor%20Object%20Reference) (Parameters :type Gameplay%20Cue%20Parameters%20Structure%20%28by%20ref%29) (SpawnResults :type Gameplay%20Cue%20Notify%20Spawn%20Result%20Structure%20%28by%20ref%29))
  (bind (_returnvalue_location _returnvalue_rotation _returnvalue_scale) (Math|Transform|BreakTransform (Transformation|GetActorTransform Target)))
  (bind _asfort_player_pawn (Utilities|Casting|CastToFortPlayerPawn Target))
  (Networking|CanExecuteCosmeticEvents
    (:True
      (bind _returnvalue (Niagara|SpawnSystematLocation 0 "/NitroGauntletsGameplay/Assets/FX/NS_NitroGauntlets_GroundPunchSmoke.NS_NitroGauntlets_GroundPunchSmoke" _returnvalue_location _returnvalue_rotation _returnvalue_scale true true "AutoRelease"))
      (Effects|Components|ParticleSystem|SetBoolParameter _returnvalue "IsGround" (AI|Components|NavMovement|IsMovingonGround (Variables|/Script/Engine.Character|GetCharacterMovement _asfort_player_pawn)))
      (bind _returnvalue_2 (Development|GetConsoleVariableIntValue "r.PhysicsField.EnableField"))
      (if (== _returnvalue_2 1)
        (Game|SpawnActorfromClass 0 "/Game/Effects/Niagara/Blueprints/PhsyicsFields/FN_RadialForce.FN_RadialForce_C" (Math|Transform|MakeTransform (+ _returnvalue_location (* (Transformation|GetActorForwardVector Target) "300.000000,300.000000,300.000000")) _returnvalue_rotation _returnvalue_scale) "Undefined" "MultiplyWithRoot" 0 0.4 0.75 4e+02 1.4 0.0 0)
        (return)))))
